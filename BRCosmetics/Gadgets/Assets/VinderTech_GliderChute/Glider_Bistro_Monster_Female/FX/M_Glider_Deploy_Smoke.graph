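# Translucent, unlit, two-sided Niagara material for smoke and trail effects on the Bistro Monster Female glider deploy.
# Scalar lerps select among default smoke, SplineMesh and UseFor_Foe looks for color, opacity and world position offset.
# Panned, noise-distorted smoke masks build opacity, with particle dynamic parameters driving panning and thickness.
# The Foe look adds UV-channel selection, depth fade, optional two-color tint, view-space sine wobble and optional SplineThicken.
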Material /BRCosmetics/Gadgets/Assets/VinderTech_GliderChute/Glider_Bistro_Monster_Female/FX/M_Glider_Deploy_Smoke
graph {
 properties {
  bUsedWithNiagaraMeshParticles=true
  bUsedWithNiagaraRibbons=true
  bUsedWithNiagaraSprites=true
  blendMode=BLEND_Translucent
  pixelDepthOffsetMode=PDOM_Legacy
  shadingModel=MSM_Unlit
  twoSided=true
  usedShadingModels=Unlit
 }
 nodes [
  n1 = TextureCoordinate() {uTiling=0.5, vTiling=0.4}
  n2 = DynamicParameter()
  n3 = AppendVector(A=Constant() {r=-0.1}, B=Multiply(A=Constant() {r=0.3}, B=@n2.Param2))
  n4 = MaterialFunctionCall(XY=TextureCoordinate() {vTiling=1.4}) {materialFunction=&Swizzle}
  n5 = Constant() {r=0.7}
  n6 = Add(A=ComponentMask(None=@n4.YX) {r=true}, B=Multiply(A=Time(), B=Multiply(A=@n2.Param2, B=@n5)) {constB=0.7})
  n7 = ComponentMask(None=@n4.YX) {g=true}
  n8 = TextureSample(UVs=Panner(Coordinate=TextureCoordinate() {uTiling=0.45, vTiling=0.37}) {speedX=0.5}) {samplerType=SAMPLERTYPE_Normal, texture=/Game/Global/Textures/Default/Noise/T_SwirlNoise_N}
  n9 = TextureSample(UVs=Add(A=AppendVector(A=@n6, B=@n7), B=Multiply(A=@n8.R) {constB=0.1})) {samplerType=SAMPLERTYPE_Grayscale, texture=&T_Smoke_Trail_Soft}
  n10 = Multiply(
   A=TextureSample(UVs=Panner(Coordinate=@n1, Speed=@n3) {speedX=-0.1, speedY=0.3}) {samplerType=SAMPLERTYPE_Masks, texture=/Game/Weapons/WeaponSkins/Wraps/Textures/Mash/Mash_Noise}.R
   B=@n9.RGB
  )
  n11 = TextureCoordinate()
  n12 = TextureSample(UVs=Panner(
    Coordinate=MaterialFunctionCall(XY=@n11) {materialFunction=&Swizzle}.YX
    Time=DynamicParameter().Param1
   ) {speedX=1}) {
   samplerSource=SSM_Clamp_WorldGroupSettings
   samplerType=SAMPLERTYPE_LinearColor
   texture=/BRCosmetics/Effects/Fort_Effects/Textures/Fire/Candle
  }
  n13 = Multiply(A=Multiply(A=@n10, B=@n12.R)) {constB=4.5}
  n14 = ParticleColor()
  n15 = TextureCoordinate()
  n16 = Multiply(A=TextureSample(UVs=TextureCoordinate() {uTiling=0.6, vTiling=0.77}) {samplerType=SAMPLERTYPE_Normal, texture=&T_SwirlNoise_N}.R) {constB=0.5}
  n17 = TextureSample(UVs=Panner(Coordinate=Add(A=@n15, B=@n16), Time=DynamicParameter().Param1) {speedX=0.3, speedY=-1}) {
   samplerSource=SSM_Clamp_WorldGroupSettings
   samplerType=SAMPLERTYPE_Masks
   texture=/Game/Global/Textures/Default/Masks_and_Gradients/T_VertGradient_MASK_2
  }
  n18 = TextureCoordinate() {uTiling=1.5}
  n19 = AppendVector(A=Constant() {r=1}, B=DynamicParameter().Param2)
  n20 = Multiply(
   A=Multiply(
    A=@n17.R
    B=TextureSampleParameter2D() {parameterName=Param, texture=/Game/Global/Textures/Default/Tileable_Patterns/T_Blowing_Smoke_Pattern2_Tile_DISP}.R
   )
   B=TextureSample(UVs=Multiply(A=@n18, B=@n19)) {texture=/Game/Accessories/FORT_Backpacks/Backpack_F_MED_Cube_Queen/FX/T_Backpack_Cube_Queen_Mask}.R
  )
  n21 = ScalarParameter() {defaultValue=1, parameterName=Dploy_Opacity}
  n22 = LinearInterpolate(
   A=TextureCoordinate() {uTiling=1.2, vTiling=0.9}
   B=TextureCoordinate() {coordinateIndex=1, uTiling=1.2, vTiling=0.9}
   Alpha=ScalarParameter() {parameterName="UV(0,1)"}
  )
  n23 = TextureCoordinate() {uTiling=0.4}
  n24 = TextureCoordinate() {coordinateIndex=1, uTiling=0.4}
  n25 = ScalarParameter() {parameterName="UV(0,1)"}
  n26 = TextureSample(UVs=Panner(Coordinate=LinearInterpolate(A=@n23, B=@n24, Alpha=@n25)) {speedX=-0.6, speedY=0.55}) {samplerType=SAMPLERTYPE_Normal, texture=&T_Weap_Distort_N_02}
  n27 = Multiply(A=@n26.R) {constB=0.13}
  n28 = TextureSample(UVs=Panner(Coordinate=Add(A=@n22, B=@n27)) {speedX=-2}) {texture=/Game/Effects/Fort_Effects/Textures/Smoke/T_Lobber_Trail_Mask}
  n29 = Constant() {r=0.1}
  n30 = TextureCoordinate() {uTiling=0.6}
  n31 = TextureCoordinate() {coordinateIndex=1, uTiling=0.6}
  n32 = ScalarParameter() {parameterName="UV(0,1)"}
  n33 = Panner(Coordinate=Add(A=LinearInterpolate(A=@n30, B=@n31, Alpha=@n32), B=Multiply(A=@n26.R) {constB=0.12})) {speedX=-1.2}
  n34 = LinearInterpolate(
   A=TextureCoordinate()
   B=TextureCoordinate() {coordinateIndex=1, uTiling=0.4, vTiling=0.8}
   Alpha=ScalarParameter() {parameterName="UV(0,1)"}
  )
  n35 = Multiply(
   A=TextureSample(UVs=Panner(Coordinate=@n34) {speedX=-0.55, speedY=0.1}) {samplerType=SAMPLERTYPE_LinearGrayscale, texture=/BRCosmetics/Weapons/FORT_Melee/Pickaxe_Cube_Ninja_Male/FX/T_FluidFlameTile}.R
   B=ScalarParameter() {defaultValue=0.32, parameterName=Fkame}
  ) {constB=0.32}
  n36 = TextureCoordinate()
  n37 = Sine(None=Multiply(A=ComponentMask(None=@n36) {g=true}) {constB=0.5})
  n38 = Multiply(
   A=Multiply(A=Add(A=@n28.R, B=@n29), B=ScalarParameter() {defaultValue=4.4})
   B=Add(A=TextureSample(UVs=@n33) {texture=/Game/Effects/Fort_Effects/Textures/Smoke/T_Smoke_Trail_Soft_2}.R, B=Multiply(A=@n35, B=@n37)) {constB=0.02}
  )
  n39 = Add(
   A=LinearInterpolate(
    A=TextureCoordinate() {uTiling=1.4}
    B=TextureCoordinate() {coordinateIndex=1, uTiling=1.4}
    Alpha=ScalarParameter() {parameterName="UV(0,1)"}
   )
   B=@n27
  )
  n40 = Power(Base=TextureSample(UVs=Panner(Coordinate=@n39) {speedX=-1.2}) {samplerType=SAMPLERTYPE_Grayscale, texture=&T_Smoke_Trail_Soft}.R) {constExponent=1.4}
  n41 = Multiply(A=@n40) {constB=0.5}
  n42 = TextureSample(UVs=@n33) {texture=/Game/Effects/Fort_Effects/Textures/Energy/T_EnergyFlow_Packed}
  n43 = LinearInterpolate(A=@n42.R, B=@n42.R, Alpha=Constant() {r=0.5})
  n44 = ComponentMask(None=LinearInterpolate(A=TextureCoordinate(), B=TextureCoordinate() {coordinateIndex=1}, Alpha=ScalarParameter() {parameterName="UV(0,1)"})) {r=true}
  n45 = OneMinus(None=@n44)
  n46 = Constant() {r=1}
  n47 = Multiply(
   A=Add(A=Add(A=@n38, B=@n41), B=Multiply(A=@n43) {constB=0.7})
   B=Multiply(A=@n45, B=SphereMask(A=@n44, B=@n46) {attenuationRadius=1, hardnessPercent=80})
  )
  n48 = ParticleColor()
  LinearInterpolate(
   A=LinearInterpolate(A=Multiply(A=@n13, B=@n14.A) {constB=4.5}, B=Multiply(A=@n20, B=@n21), Alpha=ScalarParameter() {parameterName=SplineMesh})
   B=DepthFade(Opacity=Multiply(A=@n47, B=@n48.A)) {fadeDistanceDefault=35}
   Alpha=ScalarParameter() {parameterName=UseFor_Foe}
  )
  n49 = Multiply(A=ScalarParameter() {defaultValue=15, parameterName=Param_1}, B=DynamicParameter().Param4)
  n50 = MaterialFunctionCall(WidthBase=@n49, WidthTip=@n49) {materialFunction=&SplineThicken}
  n51 = VectorParameter() {defaultValue=<rgba 0 0 1 1>}
  n52 = Multiply(A=DynamicParameter().Param2, B=@n51.B)
  n53 = Add(A=TextureCoordinate(), B=Multiply(A=Time(), B=AppendVector(A=@n52, B=@n51.A)))
  n54 = ComponentMask(None=TextureCoordinate()) {g=true}
  n55 = Multiply(
   A=Multiply(A=DynamicParameter().Param3, B=TextureSample(UVs=@n53) {texture=&T_Lobber_Trail_Mask}.RGB) {constB=5}
   B=Sine(None=Multiply(A=@n54) {constB=0.5})
  )
  n56 = TextureCoordinate()
  n57 = TextureCoordinate() {coordinateIndex=1}
  n58 = ScalarParameter() {parameterName="UV(0,1)"}
  n59 = Multiply(A=ComponentMask(None=LinearInterpolate(A=@n56, B=@n57, Alpha=@n58)) {r=true}, B=ScalarParameter() {defaultValue=1.2, parameterName=WPO_Wave}) {constB=1.2}
  n60 = Multiply(A=Time()) {constB=0.55}
  n61 = Power(Base=Sine(None=Add(A=@n59, B=@n60)), Exp=ScalarParameter() {defaultValue=3, parameterName=WPOPower}) {constExponent=3}
  n62 = VectorParameter() {parameterName=WPOSpace, defaultValue=<rgba 1 1 1 1>}
  n63 = Multiply(
   A=Transform(None=Multiply(A=@n61, B=@n62.RGB)) {transformSourceType=TRANSFORMSOURCE_View, transformType=TRANSFORM_Local}
   B=ScalarParameter() {defaultValue=4, parameterName=WPO}
  )
  n64 = ComponentMask(None=LinearInterpolate(A=TextureCoordinate(), B=TextureCoordinate() {coordinateIndex=1}, Alpha=ScalarParameter() {parameterName="UV(0,1)"})) {r=true}
  n65 = MaterialFunctionCall(
   WidthBase=ScalarParameter() {defaultValue=15, parameterName=Foe_Thicken_Base}
   WidthTip=ScalarParameter() {defaultValue=15, parameterName=Foe_Thicken_Tip}
  ) {materialFunction=&SplineThicken}
  n66 = ScalarParameter() {parameterName=Use_Thicken_Foe}
  LinearInterpolate(
   A=LinearInterpolate(A=Add(A=@n50.WorldPositionOffset, B=@n55), B=Constant(), Alpha=ScalarParameter() {parameterName=SplineMesh})
   B=Add(A=Multiply(A=@n63, B=@n64), B=LinearInterpolate(B=@n65.WorldPositionOffset, Alpha=@n66))
   Alpha=ScalarParameter() {parameterName=UseFor_Foe}
  )
  n67 = ComponentMask(None=TextureCoordinate()) {g=true}
  n68 = Time()
  n69 = Sine(None=Multiply(A=Add(A=@n67, B=@n68)) {constB=1.4})
  MaterialFunctionCall(X=@n69, Y=@n69, Z=@n69) {materialFunction=&MakeFloat3}
  MaterialFunctionCall() {materialFunction=&VectorToRadialValue}
  n70 = MaterialFunctionCall(XY=TextureCoordinate() {uTiling=0.8}) {materialFunction=&Swizzle}
  MaterialFunctionCall!(
   Base=@n9.RGB
   Blend=TextureSample(UVs=Panner(Coordinate=@n70.YX) {speedX=0.8}) {texture=&T_Lobber_Trail_Mask}.R
  ) {materialFunction=/Engine/Functions/Engine_MaterialFunctions03/Blends/Blend_PinLight}
  MaterialFunctionCall() {materialFunction=&DebugTimeSine}
  n71 = LinearInterpolate(A=Constant3Vector(), B=Constant3Vector() {constant=<rgba 0.0312 0 0.0069 0>}, Alpha=@n10)
  n72 = LinearInterpolate(
   B=VectorParameter() {parameterName=Param_1, defaultValue=<rgba 0.0417 0 0.0132 1>}.RGB
   Alpha=SphereMask(A=TextureCoordinate(), B=Constant() {r=0.5}) {attenuationRadius=0.4, hardnessPercent=50}
  )
  n73 = ScalarParameter() {parameterName=SplineMesh}
  n74 = LinearInterpolate(A=Multiply(A=ScalarParameter() {defaultValue=0.2, parameterName=Param_2}, B=@n48.RGB), B=@n48.RGB, Alpha=@n40)
  n75 = LinearInterpolate(
   A=LinearInterpolate(
    A=VectorParameter() {parameterName=Color_1, defaultValue=<rgba 0 0.2396 0.0736 1>}.RGB
    B=VectorParameter() {parameterName=Color_2, defaultValue=<rgba 0.0332 0.9115 0.0578 1>}.RGB
    Alpha=@n45
   )
   B=@n48.RGB
   Alpha=@n40
  )
  n76 = ScalarParameter() {parameterName=Mesh_Trail}
  MaterialFunctionCall(Emissive=LinearInterpolate(
    A=LinearInterpolate(A=@n71, B=@n72, Alpha=@n73)
    B=LinearInterpolate(A=@n74, B=@n75, Alpha=@n76)
    Alpha=ScalarParameter() {parameterName=UseFor_Foe}
   )) {materialFunction=&MF_FX_EmissiveAdjustments}
  Multiply(A=@n43, B=@n43)
  n77 = Constant() {r=0.5}
  SphereMask(A=TextureCoordinate(), B=@n77, Radius=@n77) {hardnessPercent=0}
  TextureCoordinate()
 ]
}
